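# Niagara system for the CraneAnchor backpack cosmetic that produces a glow and light, with user inputs for eliminations, reactivity duration and front-end state.
# Its single GlowAndLight emitter burst-spawns an offset, fixed-facing glow sprite with light and component renderers.
# System-level Color and BrightnessMultiplier drive the emitter colors and component light.
# System modules include Timeline and value-change detection, with curve, waveform and time-of-day dynamic inputs, though their wiring is not shown.
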NiagaraSystem /BRCosmetics/Accessories/FORT_Backpacks/Backpack_M_MED_CraneAnchor/FX/NS_Backpack_CraneAnchor
dependencies {
 usedDataInterfaces [&NiagaraDataInterfaceCurve]
 usedDynamicInputs [
  &Divide_Float
  &InvertBool
  &Waveform
  &Lerp_Float
  &FloatFromCurve
  &FNTimeOfDayScalar
 ]
 usedModules [
  &self:SystemSpawnScript.SystemScriptSource.SystemScriptGraph.NiagaraNodeAssignment_1.SetVariables_#1
  &SystemState
  /Game/Effects/Niagara/Modules/Logic/UpdateValueStateChange
  &Timeline
  &self:SystemSpawnScript.SystemScriptSource.SystemScriptGraph.NiagaraNodeAssignment_0.SetVariables_#2
  &self:SystemSpawnScript.SystemScriptSource.SystemScriptGraph.NiagaraNodeAssignment_2.SetVariables_#3
 ]
}
emitters {GlowAndLight {
  emitterSpawnScript [/SetVariables(Emitter.Position:NiagaraPosition=<2.5 -10 4>) {isSetParameters=true}]
  emitterUpdateScript [
   &EmitterState(
    Inactive Response=::NewEnumerator1
    Life Cycle Mode=::NewEnumerator1
    Loop Behavior=::NewEnumerator1
    Loop Duration Mode=::NewEnumerator1
    UseLoopDelay=false
   )
   &SpawnBurst_Instantaneous(Use Spawn Probability=false, UseLoopCountLimit=false)
   /SetVariables(
    Emitter.GlowColor:LinearColor=&MultiplyLinearColor(
     Linear Color=$System.Color
     Scale Factor=&MakeFloat4FromVectorAndFloat("Vector (RGB)"=&VectorFromFloat(Value=$System.BrightnessMultiplier), "Float (Alpha)"=0.05)
    )
    Emitter.LightColor:LinearColor=&MultiplyLinearColor(
     Linear Color=$System.Color
     Scale Factor=&MakeVector4_FromFloat(Float=&Multiply_Float(A=$System.BrightnessMultiplier, B=2))
    )
   ) {isSetParameters=true}
  ]
  particleSpawnScript [
   &InitializeParticle(
    Color Mode=::NewEnumerator0
    Lifetime=0
    Position Offset=<2.5 -18 4>
    Sprite Size Mode=::NewEnumerator3
    Uniform Sprite Size=90
    UsePositionOffset=true
    Write Material Random=true
   )
   &SpriteFacingAndAlignment(Sprite Facing=<0 -1 0.1>, WriteAlignment=true, WriteFacing=true)
   &CameraOffset(Camera Offset Amount=9)
   &DynamicMaterialParameters(
    Index 0 Param 1=1
    Module.Index0Param0Write=true
    Module.Index0Param1Write=false
    Module.Index0Param2Write=false
    Module.Index0Param3Write=false
   )
  ]
  particleUpdateScript [/SetVariables(
    Particles.ComponentLightColor:Vector3f=&MakeVectorFromLinearColorRGB(Color=$System.Color)
    Particles.ComponentLightIntensity:NiagaraFloat=&Multiply_Float(A=$System.BrightnessMultiplier, B=2)
   ) {isSetParameters=true}]
  rendererClasses [
   &NiagaraSpriteRendererProperties
   &NiagaraLightRendererProperties
   &NiagaraComponentRendererProperties
  ]
  renderers [
   {rendererClass=&NiagaraSpriteRendererProperties}
   {rendererClass=&NiagaraLightRendererProperties, rendererIndex=1}
   {rendererClass=&NiagaraComponentRendererProperties, rendererIndex=2}
  ]
 }}
summary {
 systemName=NS_Backpack_CraneAnchor
 emitters [{
   emitterName=GlowAndLight
   rendererClasses [
    &NiagaraSpriteRendererProperties
    &NiagaraLightRendererProperties
    &NiagaraComponentRendererProperties
   ]
  }]
 userVariables [
  User.bIsFrontEnd {type=NiagaraBool}
  User.bIsFrontEndPreview {type=NiagaraBool}
  User.Eliminations {type=NiagaraInt32}
  User.ReactivityDuration {type=NiagaraFloat, defaultValue=5, description=""}
 ]
}
