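# Material for the CraneAnchor backpack's screen, layering looping TV footage, static and noise into emissive on the 2023 character parent functions, with base color overridden to a constant.
# Flipbook video series play on aspect-corrected, tileable UVs, with a floored, modulo timer gating timed switches between footage and noise bursts.
# Static comes from three panned samples of a cinematic screen texture.
# An elimination-time parameter drives a timed reactivity mask and color.
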Material /BRCosmetics/Accessories/FORT_Backpacks/Backpack_M_MED_CraneAnchor/Materials/M_CraneAnchor_Screen
graph {
 properties {
  bUseMaterialAttributes=true
  bUsedWithMeshDeformer=true
  bUsedWithSkeletalMesh=true
  pixelDepthOffsetMode=PDOM_Legacy
  shadingModel=MSM_FromMaterialExpression
  usedShadingModels=Default Lit | Subsurface | Clear Coat | Subsurface Profile | Cloth
 }
 nodes [
  n1 = Constant() {r=12}
  n2 = Constant() {r=16}
  n3 = Frac(None=Multiply(A=Time(), B=ScalarParameter() {defaultValue=24, group=Default_Looping, parameterName=Noise_PlaySpeed}) {constB=0.2})
  n4 = Constant() {r=4}
  n5 = NamedRerouteUsage()
  n6 = TextureObjectParameter() {
   group=Default_Looping
   parameterName=NoiseVariation
   texture=/BRCosmetics/Accessories/FORT_Backpacks/Backpack_M_MED_CraneAnchor/Textures/T_NoiseFlip
  }
  n7 = MaterialFunctionCall(
   "Animation  Phase (0-1)"=LinearInterpolate(B=Divide(A=@n1, B=@n2), Alpha=@n3)
   Number of Rows=@n4
   Number of Columns=@n4
   Texture=@n6
   UVs=@n5
  ) {materialFunction=&FlipBook}
  Desaturation(None=@n7.Result, Fraction=Constant() {r=1})
  LinearInterpolate(Alpha=NamedRerouteUsage()) {constA=0.5}
  n8 = MaterialFunctionCall() {materialFunction=&MF_CharacterParent_2023_PreFX}
  n9 = GetMaterialAttributes(None=@n8.Result) {attributeGetTypes [#1, #2]}
  n10 = NamedRerouteUsage()
  n11 = Constant() {r=8}
  n12 = TextureObjectParameter() {
   group=Default_Looping
   parameterName=ScreenVariation
   texture=/BRCosmetics/Accessories/FORT_Backpacks/Backpack_M_MED_CraneAnchor/Textures/T_VideoSerie02
  }
  n13 = NamedRerouteUsage()
  n14 = LinearInterpolate(
   A=LinearInterpolate(
    A=NamedRerouteUsage()
    B=MaterialFunctionCall("Animation  Phase (0-1)"=@n10, Number of Rows=@n11, Number of Columns=@n11, Texture=@n12, UVs=@n13) {materialFunction=&FlipBook}.Result
    Alpha=NamedRerouteUsage()
   )
   B=NamedRerouteUsage()
   Alpha=NamedRerouteUsage()
  )
  n15 = NamedRerouteUsage()
  n16 = NamedRerouteUsage()
  n17 = Add(A=Add(A=LinearInterpolate(A=@n14, B=@n15, Alpha=@n16), B=NamedRerouteUsage()), B=NamedRerouteUsage())
  MaterialFunctionCall(
   Material=SetMaterialAttributes!(MaterialAttributes=@n9.MaterialAttributes, Emissive Color=Add(A=@n9.Emissive Color, B=@n17), Base Color=Constant()) {attributeSetTypes [#1, #3]}
   M=@n8.M
   Pre Skinned Local Position=@n8.M
   Distance Blend=@n8.M
   Fresnel=@n8.M
  ) {materialFunction=&MF_CharacterParent_2023_PostFX}
  n18 = NamedRerouteUsage()
  n19 = Constant2Vector() {g=0.5, r=0.5}
  n20 = VectorParameter() {group=Default_Looping, parameterName=NoiseColor_Dark}
  n21 = VectorParameter() {group=Default_Looping, parameterName=NoiseColor_Light, defaultValue=<rgba 1 1 1 0>}
  n22 = NamedRerouteUsage()
  n23 = ScalarParameter() {defaultValue=0.2, group=Default_Looping, parameterName=Tilling_Noise}
  n24 = ScalarParameter() {group=Default_Looping, parameterName=NoiseXSpeed}
  n25 = ScalarParameter() {group=Default_Looping, parameterName=NoiseYSpeed}
  n26 = TextureSampleParameter2D(UVs=Panner(Coordinate=Multiply(A=@n22, B=@n23), Speed=AppendVector(A=@n24, B=@n25)) {speedY=0.2}) {
   group=Default_Looping
   parameterName=NoiseIdle
   samplerType=SAMPLERTYPE_Masks
   texture=/BRCosmetics/Accessories/FORT_Backpacks/Backpack_M_MED_CraneAnchor/Textures/T_3MixNoise
  }
  n27 = Multiply(A=Time(), B=ScalarParameter() {defaultValue=3, group=Default_Looping, parameterName=NoiseSpeed})
  n28 = ComponentMask(None=MaterialFunctionCall(A=@n26.R, B=@n26.R, C=@n26.R, Alpha=Frac(None=@n27)) {materialFunction=&Lerp_3Color}.Result) {r=true}
  NamedRerouteDeclaration(None=Multiply(A=SphereMask(A=@n18, B=@n19) {attenuationRadius=0.6, hardnessPercent=0.5}, B=LinearInterpolate(A=@n20.RGB, B=@n21.RGB, Alpha=@n28))) {name=NoiseIdle, nodeColor=<rgba 0.9453 0 1 1>}
  n29 = MaterialFunctionCall(Float2=NamedRerouteUsage()) {materialFunction=&BreakOutFloat2Components}
  n30 = Divide(A=@n29.R, B=Divide(A=Constant() {r=1920}, B=Constant() {r=1080}))
  n31 = Add(
   A=Multiply(
    A=MaterialFunctionCall(X=@n30, Y=@n29.G) {materialFunction=&MakeFloat2}.Result
    B=ScalarParameter() {defaultValue=1.5, group=Default_Looping, parameterName=DefaultTilling}
   )
   B=ComponentMask(None=VectorParameter() {group=Default_Looping, parameterName=DefaultOffset, defaultValue=<rgba 0 0 0 1>}.RGB) {g=true, r=true}
  )
  NamedRerouteDeclaration(None=@n31) {name=TV_UV, nodeColor=<rgba 0 0.7422 1 1>}
  n32 = VectorParameter() {
   group=Default_Looping
   parameterName=ConstantNoise_Color
   channelNames {a=Intensity, b="", g="", r=""}
   defaultValue=<rgba 0.2039 0.9271 0.6979 1>
  }
  n33 = NamedRerouteUsage()
  n34 = TextureSample(UVs=Panner(Coordinate=@n33) {speedX=5, speedY=0.1}) {
   samplerSource=SSM_Wrap_WorldGroupSettings
   samplerType=SAMPLERTYPE_Masks
   texture=/Game/Missions/Primary/RepairTheShelter/Textures/Cinematic_Screen_Static
  }
  n35 = TextureSample(UVs=Panner(Coordinate=@n33) {speedX=6, speedY=-0.05}) {
   samplerSource=SSM_Wrap_WorldGroupSettings
   samplerType=SAMPLERTYPE_Masks
   texture=&Cinematic_Screen_Static
  }
  n36 = TextureSample(UVs=Panner(Coordinate=@n33) {speedX=-3, speedY=-0.03}) {
   samplerSource=SSM_Wrap_WorldGroupSettings
   samplerType=SAMPLERTYPE_Masks
   texture=&Cinematic_Screen_Static
  }
  n37 = Add(A=Multiply(A=@n34.B) {constB=0.2}, B=Multiply(A=@n35.G, B=@n36.R))
  n38 = ScalarParameter() {defaultValue=3, group=TV Glitches, parameterName=Static Scale}
  n39 = Saturate(None=NamedRerouteDeclaration(None=Multiply(A=@n37, B=@n38)) {name=Static, nodeColor=<rgba 1 0.6562 0 1>})
  NamedRerouteDeclaration(None=Multiply(A=Multiply(A=@n32, B=@n32), B=Multiply(A=@n39) {constB=0.2, desc=constant static})) {name=StaticConstant, nodeColor=<rgba 0 0.7422 1 1>}
  n40 = Fmod(A=Floor(None=Time()), B=ScalarParameter() {defaultValue=25, group=Default_Looping, parameterName=TimeGap})
  n41 = ScalarParameter() {defaultValue=15, parameterName=SecondTriggerFrame}
  n42 = NamedRerouteDeclaration(None=@n41) {name=StartFrame01, nodeColor=<rgba 0 1 0.8828 1>}
  n43 = Constant() {r=1}
  n44 = Constant()
  n45 = NamedRerouteDeclaration(None=Add(A=@n41, B=NamedRerouteUsage())) {name=EndFrame02, nodeColor=<rgba 1 0 0.7031 1>}
  n46 = Constant()
  n47 = Constant() {r=1}
  NamedRerouteDeclaration(None=Multiply(
    A=If(A=@n40, B=@n42, "A > B"=@n43, "A == B"=@n43, "A < B"=@n44) {constB=15}
    B=If(A=@n40, B=@n45, "A > B"=@n46, "A == B"=@n47, "A < B"=@n47) {constB=15}
   )) {name=RandomScary, nodeColor=<rgba 0 0.25 1 1>}
  n48 = MaterialFunctionCall(In=NamedRerouteUsage(), Contrast=Constant() {r=0.4}) {materialFunction=&CheapContrast}
  n49 = VectorParameter() {group=Default_Looping, parameterName=RoughnessColorTint, defaultValue=<rgba 0.2105 0.2354 0.1912 1>}
  NamedRerouteDeclaration(None=Multiply(A=Multiply(A=@n48.Result, B=@n49.RGB)) {constB=0.1}) {name=RoughnessInfluence, nodeColor=<rgba 1 0 0.8906 1>}
  n50 = Divide(A=Constant() {r=14}, B=Constant() {r=16})
  n51 = Constant() {r=4}
  NamedRerouteDeclaration(None=MaterialFunctionCall("Animation  Phase (0-1)"=LinearInterpolate(B=@n50, Alpha=@n3), Number of Rows=@n51, Number of Columns=@n51, Texture=@n6, UVs=@n5) {materialFunction=&FlipBook}.Result) {name=NoiseAnimation, nodeColor=<rgba 1 0 0.0234 1>}
  n52 = TextureObjectParameter() {
   group=Default_Looping
   parameterName=DefaultSerie01
   texture=/BRCosmetics/Accessories/FORT_Backpacks/Backpack_M_MED_CraneAnchor/Textures/T_VideoSerie01
  }
  NamedRerouteDeclaration(None=@n52) {name=Serie01, nodeColor=<rgba 0.6641 1 0 1>}
  n53 = ScalarParameter() {defaultValue=0.2, group=Default_Looping, parameterName=Default_PlaySpeed}
  NamedRerouteDeclaration(None=Divide(A=Constant() {r=1}, B=@n53)) {name=OneTurnTime01, nodeColor=<rgba 0.1953 0 1 1>}
  NamedRerouteDeclaration(None=@n7.Result) {name=NoiseIdleAnimation, nodeColor=<rgba 1 0 0.1875 1>}
  NamedRerouteDeclaration(None=@n40) {name=FmodTime, nodeColor=<rgba 0 1 0.3438 1>}
  n54 = NamedRerouteUsage()
  n55 = NamedRerouteUsage()
  n56 = Constant() {r=1}
  n57 = If(A=@n54, B=@n55, "A > B"=@n56, "A == B"=@n56, "A < B"=Constant()) {constB=15}
  n58 = Constant() {r=1}
  n59 = If(A=@n54, B=Add(A=@n55), "A > B"=Constant(), "A == B"=@n58, "A < B"=@n58) {constB=15}
  n60 = NamedRerouteUsage()
  n61 = NamedRerouteUsage()
  n62 = Constant() {r=1}
  n63 = If(A=@n60, B=Subtract(A=@n61) {constB=0.3}, "A > B"=@n62, "A == B"=@n62, "A < B"=Constant()) {constB=15}
  n64 = Constant() {r=1}
  n65 = If(A=@n60, B=@n61, "A > B"=Constant(), "A == B"=@n64, "A < B"=@n64) {constB=15}
  NamedRerouteDeclaration(None=Add(A=Multiply(A=@n57, B=@n59), B=Multiply(A=@n63, B=@n65))) {name=NoiseState02, nodeColor=<rgba 1 0.0234 0 1>}
  n66 = Subtract(A=Time(), B=ScalarParameter() {group=Reactivity, parameterName=ElimationTime})
  n67 = ScalarParameter() {defaultValue=5, group=Reactivity, parameterName=ReactivityTime}
  n68 = OneMinus(None=Saturate(None=Divide(A=@n66, B=@n67)))
  NamedRerouteDeclaration(None=Sign(None=@n68)) {name=ReactivityAlpha, nodeColor=<rgba 0.1719 1 0 1>}
  n69 = Step(X=@n68) {constY=0.1}
  n70 = Step(Y=@n68) {constX=0.9, constY=0.1}
  NamedRerouteDeclaration(None=LinearInterpolate(A=NamedRerouteUsage(), B=NamedRerouteUsage(), Alpha=Multiply(A=@n69, B=@n70))) {name=ReactivityColor, nodeColor=<rgba 1 0 0.4922 1>}
  NamedRerouteDeclaration(None=@n9.Roughness) {name=RoughnessIN, nodeColor=<rgba 0.8516 1 0 1>}
  n71 = Multiply(A=Time(), B=@n53) {constB=0.2}
  n72 = Constant() {r=8}
  NamedRerouteDeclaration(None=MaterialFunctionCall("Animation  Phase (0-1)"=@n71, Number of Rows=@n72, Number of Columns=@n72, Texture=@n52, UVs=@n31) {materialFunction=&FlipBook}.Result) {name=ScaryScreenColor01, nodeColor=<rgba 0 1 0.6953 1>}
  NamedRerouteDeclaration(None=@n71) {name=PlaySpeed_TV, nodeColor=<rgba 0 1 0.1328 1>}
  NamedRerouteDeclaration(None=@n39) {name=SubtleNoiseAlpha, nodeColor=<rgba 1 0.5859 0 1>}
  NamedRerouteDeclaration(None=LinearInterpolate(
    A=TextureCoordinate()
    B=TextureCoordinate() {coordinateIndex=1}
    Alpha=ScalarParameter() {defaultValue=1, group=Default_Looping, parameterName=UseUV1}
   )) {name=UVs, nodeColor=<rgba 1 0.4453 0 1>}
  NamedRerouteDeclaration(None=@n32) {name=NoiseTint, nodeColor=<rgba 0.8281 0 1 1>}
  NamedRerouteUsage()
 ]
}
