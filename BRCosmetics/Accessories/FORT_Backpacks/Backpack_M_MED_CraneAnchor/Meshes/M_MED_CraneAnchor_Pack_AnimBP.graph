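# Animation Blueprint for the CraneAnchor cosmetic backpack, adjusting the pack to body bend and simulating its two antennas.
# Each update maps owner-mesh spine_05-to-spine_01 delta rotation into pack rotation and translation, with a different pitch mapping while skydiving.
# An additive Modify Bone applies these, then two spring-driven AnimDynamics nodes sway the antennas.
# A second AnimDynamics pair is disconnected; computed gravity override and frontend alpha are unused in the AnimGraph.
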AnimBlueprint /BRCosmetics/Accessories/FORT_Backpacks/Backpack_M_MED_CraneAnchor/Meshes/M_MED_CraneAnchor_Pack_AnimBP
target_skeleton=&/M_MED_CraneAnchor_Pack_Skeleton
graphs [
 AnimGraph {nodes [
   n1 = Anim Dynamics - Bone: dyn_antenne_A_base(Alpha=1, ExternalForce="0,0,0") {
    angularDampingOverride=0.999
    angularSpringConstant=1000
    bAngularSpring=true
    bGravityOverrideInSimSpace=true
    bOverrideAngularDamping=true
    gravityScale=0.6
    simulationSpace=BoneRelative
    alphaBoolBlend {blendOption=Linear}
    boundBone {boneName=dyn_antenne_A_base}
    componentAppliedLinearAccClamp=<2000 2000 2000>
    componentLinearAccScale=<1 1 1>
    componentLinearVelScale=<1 1 1>
    physicsBodyDefinitions [{
      sphereCollisionRadius=0
      boundBone {boneName=dyn_antenne_A_base}
      boxExtents=<3 10 15>
      constraintSetup {angularTargetAxis=AxisY, angularLimitsMax=<5 5 5>, angularLimitsMin=<-5 -5 -5>, angularTarget=<0 0 0>}
      localJointOffset=<-4 -0.05 -1>
     }]
   }
   Anim Dynamics - Bone: dyn_antenne_B_base(ComponentPose=@n1, Alpha=1, ExternalForce="0,0,0") {
    angularDampingOverride=0.999
    angularSpringConstant=1000
    bAngularSpring=true
    bGravityOverrideInSimSpace=true
    bOverrideAngularDamping=true
    gravityScale=0.6
    simulationSpace=BoneRelative
    alphaBoolBlend {blendOption=Linear}
    boundBone {boneName=dyn_antenne_B_base}
    componentAppliedLinearAccClamp=<2000 2000 2000>
    componentLinearAccScale=<1 1 1>
    componentLinearVelScale=<1 1 1>
    physicsBodyDefinitions [{
      sphereCollisionRadius=0
      boundBone {boneName=dyn_antenne_B_base}
      boxExtents=<3 10 15>
      constraintSetup {angularTargetAxis=AxisY, angularLimitsMax=<5 10 10>, angularLimitsMin=<-5 -15 -10>, angularTarget=<0 0 0>}
      localJointOffset=<-4 -0.05 1>
     }]
   }
   n2 = "Transform (Modify) Bone - Bone: pack"!(
    ComponentPose=Local To Component(LocalPose=Input Pose())
    bAlphaBoolEnabled:bool
    Translation=Get backpack_translation()
    Rotation=Get backpack_rotate()
    Scale
   ) -> (Pose) {
    alphaInputType=Bool
    rotationMode=BMM_Additive
    scaleMode=BMM_Additive
    scaleSpace=BCS_BoneSpace
    translationMode=BMM_Additive
    alphaBoolBlend {blendInTime=0.2, blendOption=Linear, blendOutTime=0.8}
    boneToModify {boneName=pack}
   }
   n3 = Anim Dynamics - Bone: dyn_antenne_B_base(ComponentPose=@n2, Alpha=1, ExternalForce="0,0,0") {
    angularDampingOverride=0.9999
    angularSpringConstant=10000
    bAngularSpring=true
    bOverrideAngularDamping=true
    bUseGravityOverride=true
    bUsePlanarLimit=false
    gravityScale=0.35
    simulationSpace=RootRelative
    alphaBoolBlend {blendOption=Linear}
    boundBone {boneName=dyn_antenne_B_base}
    componentAppliedLinearAccClamp=<1000 1000 1000>
    componentLinearAccScale=<0.6 0.5 0.5>
    componentLinearVelScale=<0.6 0.5 0.5>
    gravityOverride {z=-980}
    physicsBodyDefinitions [{
      boundBone {boneName=dyn_antenne_B_base}
      boxExtents=<25 25 25>
      constraintSetup {angularLimitsMax=<0 22 18>, angularLimitsMin=<0 -18 -22>, angularTarget=<1 0 0>}
      localJointOffset=<1 0 0>
     }]
   }
   n4 = Anim Dynamics - Bone: dyn_antenne_A_base(ComponentPose=@n3, Alpha=1, ExternalForce="0,0,0") {
    angularDampingOverride=0.9999
    angularSpringConstant=10000
    bAngularSpring=true
    bOverrideAngularDamping=true
    bUseGravityOverride=true
    bUsePlanarLimit=false
    gravityScale=0.35
    simulationSpace=RootRelative
    alphaBoolBlend {blendOption=Linear}
    boundBone {boneName=dyn_antenne_A_base}
    componentAppliedLinearAccClamp=<1000 1000 1000>
    componentLinearAccScale=<0.6 0.5 0.5>
    componentLinearVelScale=<0.6 0.5 0.5>
    gravityOverride {z=-980}
    physicsBodyDefinitions [{
      boundBone {boneName=dyn_antenne_A_base}
      boxExtents=<20 20 20>
      constraintSetup {angularLimitsMax=<0 18 22>, angularLimitsMin=<0 -22 -18>, angularTarget=<1 0 0>}
      localJointOffset=<1 0 0>
     }]
   }
   Output Pose(Result=Component To Local(ComponentPose=@n4)) {layerGroup=None}
  ]}
 EventGraph {nodes [
   Event BlueprintUpdateAnimation()
   > Is Valid(InputObject=Get BodyMesh() -> (BodyMesh:Skeletal Mesh Component Object Reference))
   > n1 = Gravity Override State Function DEPRECATED(CharacterPartAnimInstance=Self-Reference(), GameTimeInSeconds=GetGameTimeInSeconds(), GravityOverrideParams=Get Gravity Override Params())
   > Set Gravity Override(Gravity Override=@n1)
   n2 = GetOwnerSkeletalMeshComponent()
   n3 = BreakRotator(InRot=BreakTransform(InTransform=GetDeltaTransformFromRefPose(self=@n2, BoneName=spine_05, BaseName=spine_01)).Rotation)
   n4 = MapRangeClamped(Value=@n3.Pitch, InRangeA=30, InRangeB=-30, OutRangeA=-30, OutRangeB=30)
   n5 = MapRangeClamped(Value=@n3.Pitch, InRangeA=0, InRangeB=-30, OutRangeA=-5, OutRangeB=25)
   n6 = OR Boolean(A=Get bIsSkydiveFloating(), B=Get bIsSkyDiving())
   > Set backpack_rotate(backpack_rotate=MakeRotator(Roll=MapRangeClamped(Value=@n3.Yaw, InRangeA=0, InRangeB=-30, OutRangeA=5, OutRangeB=25), Pitch=SelectFloat(A=@n5, B=@n4, bPickA=@n6)))
   > Set backpack_translation(backpack_translation=MakeVector(Y=MapRangeClamped(Value=@n3.Yaw, InRangeA=0, InRangeB=-30, OutRangeB=-6.5)))
   > Set frontendAlphaControl(frontendAlphaControl:double=SelectFloat(A=0.25, B=0.8, bPickA=Get bIsInFrontEnd())) -> (Output_Get:double)
   n7 = Event OnCharacterCustomizationLoaded()
   > Set BodyMesh(BodyMesh=Get Array Item with EFort Custom Part Type(Array=@n7.CustomMeshParts))
   TryGetPawnOwner()
   TryGetPawnOwner()
  ]}
]
variables [
 " jump alpha pack" {type=number}
 backpack_rotate {type=Rotator}
 backpack_translation {type=Vector}
 BodyMesh {type=SkeletalMeshComponent}
 frontendAlphaControl {type=number}
 Gravity Override {type=Vector}
 Gravity Override Params {type=GravityOverrideParamsStruct}
 spine 04 rotation {type=Rotator}
 spine 04 translation {type=Vector}
 spine_01 translation {type=Vector}
 spine_01rotation {type=Rotator}
]
